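# Post-process AnimBP for the BitterDolly biped cosmetic companion that adds secondary motion to its coat tails.
# The input pose passes through chained Anim Dynamics nodes for left and right tail bones, using root-relative simulation, angular springs, limits and reduced gravity.
# Both nodes share DynamicsAlpha, inverted by scale and bias, so higher values weaken the simulation.
# The EventGraph is empty, and the template PoseAsset variable is unused in the shown graphs.
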AnimBlueprint /CosmeticCompanions/Assets/Biped/BitterDolly/Meshes/BitterDolly_Biped_Post_AnimBP
target_skeleton=&/BitterDolly_Biped_Skeleton
graphs [
 AnimGraph {nodes [
   n1 = Get DynamicsAlpha() -> (DynamicsAlpha:float)
   n2 = Anim Dynamics - Bone: dyn_coatTail_l_01(ComponentPose=Local To Component(LocalPose=Input Pose()), Alpha=@n1, ExternalForce="0,0,0") {
    angularDampingOverride=1
    angularSpringConstant=12000
    bAngularSpring=true
    bOverrideAngularDamping=true
    bUseGravityOverride=true
    gravityScale=0.35
    simulationSpace=RootRelative
    alphaScaleBiasClamp {bias=1, scale=-1}
    boundBone {boneName=dyn_coatTail_l_01}
    componentAppliedLinearAccClamp=<2000 2000 2000>
    componentLinearAccScale=<1 1 1>
    componentLinearVelScale=<1 1 1>
    gravityOverride {z=-980}
    physicsBodyDefinitions [{
      boundBone {boneName=dyn_coatTail_l_01}
      boxExtents=<20 30 30>
      constraintSetup {angularLimitsMax=<3 5 45>, angularLimitsMin=<-3 -5 -45>, angularTarget=<1 0 0>}
      localJointOffset=<5 0 0>
     }]
   }
   n3 = Anim Dynamics - Bone: dyn_coatTail_r_01(ComponentPose=@n2, Alpha=@n1, ExternalForce="0,0,0") {
    angularDampingOverride=1
    angularSpringConstant=10000
    bAngularSpring=true
    bOverrideAngularDamping=true
    bUseGravityOverride=true
    gravityScale=0.35
    simulationSpace=RootRelative
    alphaScaleBiasClamp {bias=1, scale=-1}
    boundBone {boneName=dyn_coatTail_r_01}
    componentAppliedLinearAccClamp=<2000 2000 2000>
    componentLinearAccScale=<1 1 1>
    componentLinearVelScale=<1 1 1>
    gravityOverride {z=-980}
    physicsBodyDefinitions [{
      boundBone {boneName=dyn_coatTail_r_01}
      boxExtents=<20 30 30>
      constraintSetup {angularLimitsMax=<3 5 45>, angularLimitsMin=<-3 -5 -45>, angularTarget=<1 0 0>}
      localJointOffset=<4 0 0>
     }]
   }
   Output Pose(Result=Component To Local(ComponentPose=@n3))
  ]}
 EventGraph {links [], nodes []}
]
variables [PoseAsset {type=PoseAsset, default=/CosmeticCompanions/Automation_Template/companion_pose_asset_template}]
